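# M_Pickaxe_PlantStand_Blob is an unlit translucent Niagara sprite material that draws a wobbly blob for the PlantStand pickaxe FX.
# Panning, tiled blurred noise offsets the UVs of a centered SphereMask, distorting the circle procedurally.
# A SmoothStep with a Hard Edge parameter shapes the mask, which is multiplied by particle alpha.
# Particle color RGB feeds the MF_FX_EmissiveAdjustments function.
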Material /BRCosmetics/Weapons/FORT_Melee/Pickaxe_PlantStand/FX/M_Pickaxe_PlantStand_Blob
graph {
 properties {bUsedWithNiagaraSprites=true, blendMode=BLEND_Translucent, pixelDepthOffsetMode=PDOM_Legacy, shadingModel=MSM_Unlit, usedShadingModels=Unlit}
 nodes [
  n1 = ParticleColor()
  MaterialFunctionCall(Emissive=@n1.RGB) {materialFunction=&MF_FX_EmissiveAdjustments}
  n2 = Constant()
  n3 = ScalarParameter() {parameterName=Hard Edge, sliderMax=0.5}
  n4 = Constant() {r=1}
  n5 = TextureCoordinate()
  n6 = AppendVector(
   A=ScalarParameter() {defaultValue=0.5, parameterName=Tile X, sliderMax=2}
   B=ScalarParameter() {defaultValue=0.5, parameterName=Tile Y, sliderMax=2}
  )
  n7 = ScalarParameter() {parameterName=Speed X, sliderMax=2}
  n8 = ScalarParameter() {defaultValue=0.5, parameterName=Speed Y, sliderMax=2}
  n9 = TextureSample(UVs=Panner(Coordinate=Multiply(A=@n5, B=@n6) {constB=0.5}, Speed=AppendVector(A=@n7, B=@n8)) {speedY=0.25}) {samplerType=SAMPLERTYPE_Grayscale, texture=/FortUILibrary/UIKit/Buttons/Textures/LowResBlurredNoise}
  n10 = ScalarParameter() {defaultValue=-0.5, parameterName=Distortion Offset, sliderMax=1, sliderMin=-1}
  n11 = Add(
   A=TextureCoordinate()
   B=Multiply(A=Subtract(A=@n9.R, B=@n10) {constB=0.5}, B=ScalarParameter() {defaultValue=0.1, parameterName=Distortion Strength, sliderMax=2}) {constB=0.2}
  )
  n12 = Constant() {r=0.5}
  n13 = Constant() {r=0.4}
  n14 = ScalarParameter() {parameterName=Hardness, sliderMax=1}
  Multiply(
   A=@n1.A
   B=SmoothStep(Min=Add(A=@n2, B=@n3), Max=Subtract(A=@n4, B=@n3), Value=SphereMask(A=@n11, B=@n12, Radius=@n13, Hardness=@n14) {hardnessPercent=0}) {constMax=0.8, constMin=0.2}
  )
 ]
}
